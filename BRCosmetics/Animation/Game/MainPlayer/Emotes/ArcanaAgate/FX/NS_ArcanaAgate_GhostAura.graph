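# Niagara system for the ArcanaAgate emote that wraps the gadget prop in a greenish ghostly aura.
# Both emitters sample the attached skeletal mesh at JNT_Frame over a 3.4-second window.
# FrameAura_Building keeps growing, faint meshes on the frame, while Aura_Trail spawns drifting, drag-slowed SubUV sprites.
# Each has two renderers and a separate BackplateColor, tinted and scaled by system-level time-of-day and fade variables.
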NiagaraSystem /BRCosmetics/Animation/Game/MainPlayer/Emotes/ArcanaAgate/FX/NS_ArcanaAgate_GhostAura
dependencies {
 usedDataInterfaces [
  &NiagaraDataInterfaceCurve
  &NiagaraDataInterfaceSkeletalMesh
  &NiagaraDataInterfaceColorCurve
  &NiagaraDataInterfaceSpriteRendererInfo
 ]
 usedDynamicInputs [&FNTimeOfDayScalar, &FNTimeOfDayColor]
 usedModules [
  &SystemState
  &self:SystemSpawnScript.SystemScriptSource.SystemScriptGraph.NiagaraNodeAssignment_0.SetVariables_#1
  &self:SystemSpawnScript.SystemScriptSource.SystemScriptGraph.NiagaraNodeAssignment_1.SetVariables_#2
 ]
}
emitters {
 Aura_Trail {
  emitterUpdateScript [
   &EmitterState(Life Cycle Mode=::NewEnumerator1, Loop Behavior=::NewEnumerator1, Loop Duration=3.4, UseLoopDelay=false)
   &SpawnRate(
    Spawn Probability=&FloatFromCurve(
     FloatCurve=DataInterface() {propertyValues {Curve {keys [<at 0 0 RCIM_Constant>,<at 0.27 0.8>,<at 1 0.8>]}}}
     CurveIndex=$Emitter.NormalizedLoopAge
    )
    SpawnRate=&FloatFromCurve(
     FloatCurve=DataInterface() {propertyValues {Curve {keys [<at 0 1 RCIM_Constant>,<at 0.95 0>,<at 1 0 RCIM_Constant>]}}}
     CurveIndex=$Emitter.NormalizedLoopAge
     Scale Curve=15.3
    )
    Use Spawn Probability=true
   )
  ]
  particleSpawnScript [
   &InitializeParticle(
    AdjustAlpha=true
    AdjustHue=false
    AdjustSaturation=false
    AdjustValue=true
    Alpha Scale Range:Vector2f=<0.7 1>
    Color=<rgba 0.2487 0.5417 0.2031 0.5>
    Color Mode=::NewEnumerator3
    Lifetime Max=1.65
    Lifetime Min=0.9
    Lifetime Mode=::NewEnumerator1
    Sprite Rotation Mode=::NewEnumerator2
    Sprite Size Mode=::NewEnumerator1
    Sprite UV Mode=::NewEnumerator3
    Uniform Sprite Size Max=80
    Uniform Sprite Size Min=50
    UsePositionOffset=false
    Write Material Random=true
   )
   &SkeletalMeshLocation(
    Bone Sampling Mode=::NewEnumerator4
    MaskSpawnGroup=false
    Skeletal Mesh=DataInterface() {propertyValues {
      DefaultMesh=/BRCosmetics/Animation/Props/Gadgets_ArcanaAgate/Meshes/ArcanaAgate_Gadget
      ExcludeBoneName=root
      OptionalComponentClass=null
      PreviewMesh=&ArcanaAgate_Gadget
      SkinningMode=SkinOnTheFly
      SoftSourceActor=null
      SourceMode=AttachParent
      UvSetIndex=0
      WholeMeshLOD=-1
      bExcludeBone=false
      bReadDeformedGeometry=true
      bRequireCurrentFrameData=true
      ComponentTags []
      FilteredBones [JNT_Frame]
      FilteredSockets []
      MeshUserParameter {parameter {name=None, typeDefHandle {registeredTypeIndex=1}, varData []}}
      SamplingRegions []
     }}
   )
   &ShapeLocation(
    Box Size=<3 30 45>
    MaskSpawnGroup=false
    Override Seed=false
    Rotation Angle=-0.03
    Rotation Axis=<1 1 0>
    Rotation Mode=::NewEnumerator1
    Shape Origin=$Particles.SkeletalMeshLocation.SampledPosition
    Shape Primitive=::NewEnumerator2
    Surface Only=false
    UseOwnerScale=true
   )
   &SubUVAnimation(
    Evaluation Type=::NewEnumerator0
    "Sprite Renderer "=DataInterface() {propertyValues {SpriteRenderer=&self:Aura_Trail_0.NiagaraSpriteRendererProperties_0}}
    SubUV Animation Mode:ENiagara_SubUVLookupModeV2=::NewEnumerator2
    UseEndFrame=false
    UseStartFrame=false
   )
   &AddVelocity(
    Module.Use Falloff Distance=false
    Origin Offset=<0 5 1>
    Override Seed=false
    Velocity Mode=::NewEnumerator1
    Velocity Speed=&RandomRangeFloat(Minimum=20, Maximum=50)
   )
   &ApplyInitialForces()
  ]
  particleUpdateScript [
   &ParticleState()
   &ScaleSpriteSize(
    Uniform Curve Scale=0.7
    Uniform Curve Sprite Scale=DataInterface() {propertyValues {Curve {defaultValue=1, keys [{tangentMode=RCTM_User, time=0, value=0.5}, {tangentMode=RCTM_User, time=1, value=1}]}}}
   )
   &DynamicMaterialParameters(
    Index 0 Param 1=&RandomRangeFloat(Minimum=0, Maximum=111)
    Module.Index0Float0Write=true
    Module.Index0Float1Write=false
    Module.Index0Float2Write=false
    Module.Index0Float3Write=false
   )
   &ScaleColor(
    Linear Color Curve=DataInterface() {propertyValues {
      AlphaCurve {keys [<at 0 0>,<at 0.2117 0.7707>,<at 0.9483 0>]}
      BlueCurve {keys [<at 0 1>,<at 1 1>]}
      GreenCurve {keys [<at 0 1>,<at 1 1>]}
      RedCurve {keys [<at 0 1>,<at 1 1>]}
     }}
    Scale Mode=::NewEnumerator2
   )
   &GravityForce(Gravity=&RandomRangeVector(Minimum=<0 0 25>, Maximum=<0 0 -55>))
   &PointForce(
    Force Origin:NiagaraPosition=$Particles.SkeletalMeshLocation.SampledPosition
    Force Strength=&RandomRangeFloat(Minimum=7, Maximum=15)
    Module.Use Falloff=false
    Module.UseLinearForce=false
   )
   &Drag(
    Drag=&FloatFromCurve(
     FloatCurve=DataInterface() {propertyValues {Curve {
        defaultValue=1
        keys [
         {arriveTangent=1.7487, interpMode=RCIM_Cubic, leaveTangent=1.7487, tangentMode=RCTM_User, time=0, value=0}
         {arriveTangent=0.0224, interpMode=RCIM_Cubic, leaveTangent=0.0224, tangentMode=RCTM_User, time=1, value=1}
        ]
       }}}
     Scale Curve=&RandomRangeFloat(Minimum=2, Maximum=4)
    )
    Use Linear Drag=true
    Use Rotational Drag=true
   )
   &SolveForcesAndVelocity(Clamp Velocity=false, Limit Acceleration=false)
   /SetVariables(
    Particles.BackplateColor:LinearColor=&Multiply_LinearColorByFloat(
     Linear Color=&MultiplyLinearColor(
      Linear Color=<rgba 0.0579 0.0885 0.0475 0.7>
      Scale Factor=&MakeVector4_FromLinearColor(LinearColor=$System.ToD_Hue)
     )
     Float=&Multiply_Float(
      A=&FloatFromCurve(
       FloatCurve=DataInterface() {propertyValues {Curve {
          defaultValue=0
          keys [
           {arriveTangent=2.3405, interpMode=RCIM_Cubic, leaveTangent=2.3405, tangentMode=RCTM_User, time=0, value=0}
           {interpMode=RCIM_Cubic, tangentMode=RCTM_User, time=0.35, value=1}
           {tangentMode=RCTM_User, time=1, value=0}
          ]
         }}}
       Scale Curve=&Multiply_Float(A=&RandomRangeFloat(Minimum=0.7, Maximum=1), B=$System.ToD)
      )
      B=$System.BackplateFadeOUt
     )
    )
    Particles.Color=&MultiplyLinearColor(Linear Color=$Particles.Color, Scale Factor=&MakeVector4_FromLinearColor(LinearColor=$System.ToD_Hue))
   ) {isSetParameters=true}
   &CameraOffset(Camera Offset Amount=15)
  ]
  rendererClasses [&NiagaraSpriteRendererProperties]
  renderers [
   {rendererClass=&NiagaraSpriteRendererProperties}
   {rendererClass=&NiagaraSpriteRendererProperties, rendererIndex=1}
  ]
 }
 FrameAura_Building {
  emitterUpdateScript [
   &EmitterState(Life Cycle Mode=::NewEnumerator1, Loop Behavior=::NewEnumerator1, Loop Duration=3.4, UseLoopDelay=false)
   &SpawnRate(
    SpawnRate=&FloatFromCurve(
     FloatCurve=DataInterface() {propertyValues {Curve {
        defaultValue=1
        keys [
         {arriveTangent=-1.0932, interpMode=RCIM_Cubic, leaveTangent=-1.0932, tangentMode=RCTM_User, time=0, value=4}
         {arriveTangent=-0.0849, interpMode=RCIM_Cubic, leaveTangent=-0.0849, tangentMode=RCTM_User, time=0.4978, value=4.0594}
         {arriveTangent=-2.841, interpMode=RCIM_Cubic, leaveTangent=-2.841, tangentMode=RCTM_User, time=2.8, value=4}
         {arriveTangent=-8.506, interpMode=RCIM_Cubic, leaveTangent=-8.506, tangentMode=RCTM_User, time=3, value=0}
        ]
       }}}
     CurveIndex=$Engine.System.Age
    )
    Use Spawn Probability=false
   )
  ]
  particleSpawnScript [
   /SetVariables(Particles.VisibilityTag=&SpawnGroup()) {isSetParameters=true}
   &InitializeParticle(
    Lifetime=&RandomRangeFloat(Minimum=2, Maximum=4)
    Mesh Scale Mode=::NewEnumerator3
    Mesh Uniform Scale=&RandomRangeFloat(Minimum=0.09, Maximum=0.19)
    UsePositionOffset=false
    Write Material Random=true
   )
   &InitialMeshOrientation(Orientation Axis=<0 0 1>, Orientation Vector=<0 0 1>, Rotation=<0 0.25 0>, Use Orientation Vector=true, Use Rotation Vector=true) {module=InitialMeshOrientation001}
   &InitialMeshOrientation(Rotation=&RandomVector(Vector Scale=0.15), Use Orientation Vector=true, Use Rotation Vector=true) {enabled=false}
   &ShapeLocation(MaskSpawnGroup=false, Override Seed=false, Sphere Radius=5, UseOwnerScale=true) {module=ShapeLocation001}
  ]
  particleUpdateScript [
   &ParticleState()
   &SkeletalMeshLocation(
    Bone Sampling Mode=::NewEnumerator4
    MaskSpawnGroup=false
    Orientation Sampling=::NewEnumerator1
    Skeletal Mesh=DataInterface() {propertyValues {
      DefaultMesh=&ArcanaAgate_Gadget
      ExcludeBoneName=root
      OptionalComponentClass=null
      PreviewMesh=&ArcanaAgate_Gadget
      SkinningMode=SkinOnTheFly
      SoftSourceActor=null
      SourceMode=AttachParent
      UvSetIndex=0
      WholeMeshLOD=-1
      bExcludeBone=false
      bReadDeformedGeometry=true
      bRequireCurrentFrameData=true
      ComponentTags []
      FilteredBones [JNT_Frame]
      FilteredSockets []
      MeshUserParameter {parameter {name=None, typeDefHandle {registeredTypeIndex=1}, varData []}}
      SamplingRegions []
     }}
   )
   &ShapeLocation(
    MaskSpawnGroup=false
    Override Seed=false
    Shape Origin=$Particles.SkeletalMeshLocation.SampledPosition
    Sphere Radius=10
    UseOwnerScale=true
   )
   &ScaleMeshSize(Scale Factor=&VectorFromFloat(Value=&FloatFromCurve(
      FloatCurve=DataInterface() {propertyValues {Curve {keys [
          {arriveTangent=4.496, interpMode=RCIM_Cubic, leaveTangent=4.496, tangentMode=RCTM_User, time=0, value=1}
          {arriveTangent=0.625, interpMode=RCIM_Cubic, leaveTangent=0.625, time=0.3, value=1.625}
          <at 1 1.625 RCIM_Cubic>
         ]}}}
      Scale Curve=4
     )))
   &ScaleColor(
    Linear Color Curve=DataInterface() {propertyValues {
      AlphaCurve {keys [<at 0 0>,<at 0.1051 0.03>,<at 1 0>]}
      BlueCurve {keys [<at 0.0029 0.228>,<at 0.2356 0.2963>,<at 0.6139 0.0714>,<at 1 0.0225>]}
      GreenCurve {keys [<at 0.0029 0.7396>,<at 0.2356 0.5521>,<at 0.6139 0.2083>,<at 1 0.0417>]}
      RedCurve {keys [<at 0.0029 0.2196>,<at 0.2356 0.3657>,<at 0.6139 0.0608>,<at 1 0.008>]}
     }}
    Scale Mode=::NewEnumerator2
   )
   &DynamicMaterialParameters(
    Index 0 Param 1=&Add_Float(
     A=&FloatFromCurve(
      FloatCurve=DataInterface() {propertyValues {Curve {keys [{arriveTangent=4.8237, interpMode=RCIM_Cubic, leaveTangent=4.8237, tangentMode=RCTM_User, time=0, value=0}, <at 1 2.25 RCIM_Cubic>]}}}
      Scale Curve=0.5
     )
     B=&RandomRangeFloat(Minimum=0, Maximum=999)
    )
    Module.Index0Param0Write=true
    Module.Index0Param1Write=false
    Module.Index0Param2Write=false
    Module.Index0Param3Write=false
   )
   &Drag(Use Linear Drag=true, Use Rotational Drag=true)
   &GravityForce(Gravity=&MaskVectorBySpawnGroup(
     Masked Vector=&RandomRangeVector(Minimum=<15 15 25>, Maximum=<-15 -15 -25>)
     Passthrough Vector=<0 0 0>
    ))
   &SolveForcesAndVelocity(Clamp Velocity=false, Limit Acceleration=false)
   &CameraOffset(Camera Offset Amount=&RandomRangeFloat(Minimum=-5, Maximum=-20)) {module=CameraOffset001}
   &ScaleColor(
    Scale Alpha=$System.FadeOut
    Scale RGB=&VectorFromFloat(Value=$System.ColorRamp)
    ScaleA=true
    ScaleRGB=true
   ) {module=ScaleColor001}
   /SetVariables(
    Particles.BackplateColor:LinearColor=&Multiply_LinearColorByFloat(
     Linear Color=&MultiplyLinearColor(
      Linear Color=<rgba 0.0579 0.0885 0.0475 0.1>
      Scale Factor=&MakeVector4_FromLinearColor(LinearColor=$System.ToD_Hue)
     )
     Float=&Multiply_Float(
      A=&FloatFromCurve(
       FloatCurve=DataInterface() {propertyValues {Curve {
          defaultValue=0
          keys [
           {arriveTangent=2.3405, interpMode=RCIM_Cubic, leaveTangent=2.3405, tangentMode=RCTM_User, time=0, value=0}
           {interpMode=RCIM_Cubic, tangentMode=RCTM_User, time=0.2195, value=1}
           {tangentMode=RCTM_User, time=1, value=0}
          ]
         }}}
       Scale Curve=&Multiply_Float(A=1, B=$System.ToD)
      )
      B=$System.BackplateFadeOUt
     )
    )
    Particles.Color=&MultiplyLinearColor(Linear Color=$Particles.Color, Scale Factor=&MakeVector4_FromLinearColor(LinearColor=$System.ToD_Hue))
    Particles.DMP1:Vector4f=&MakeVector4(X=0, Y=0, Z=0, W=0)
   ) {isSetParameters=true}
  ]
  rendererClasses [&NiagaraMeshRendererProperties]
  renderers [
   {rendererClass=&NiagaraMeshRendererProperties, rendererIndex=0}
   {rendererClass=&NiagaraMeshRendererProperties, rendererIndex=1}
  ]
 }
}
summary {
 systemName=NS_ArcanaAgate_GhostAura
 emitters [
  {emitterName=FrameAura_Building, rendererClasses [&NiagaraMeshRendererProperties]}
  {emitterName=Aura_Trail, rendererClasses [&NiagaraSpriteRendererProperties]}
 ]
 userVariables []
}
