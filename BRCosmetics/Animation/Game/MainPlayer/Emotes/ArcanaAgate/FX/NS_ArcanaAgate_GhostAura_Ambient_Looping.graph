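# Looping Niagara ambient ghost aura for the ArcanaAgate emote, built from soft green mesh particles.
# Emitter Flow001 combines a burst with steady spawning of randomly scaled meshes around a shape volume.
# Particles grow, drift under spawn-group-masked random gravity and drag, and feed an animated, random-offset value to dynamic material parameters.
# Two mesh renderers draw them, with particle and separate backplate colors tinted by time of day and faded by System.FadeOut.
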NiagaraSystem /BRCosmetics/Animation/Game/MainPlayer/Emotes/ArcanaAgate/FX/NS_ArcanaAgate_GhostAura_Ambient_Looping
dependencies {
 usedDataInterfaces [&NiagaraDataInterfaceCurve, &NiagaraDataInterfaceColorCurve]
 usedDynamicInputs [&FNTimeOfDayScalar, &FNTimeOfDayColor]
 usedModules [
  &SystemState
  &self:SystemSpawnScript.SystemScriptSource.SystemScriptGraph.NiagaraNodeAssignment_1.SetVariables_#1
 ]
}
emitters {Flow001 {
  emitterUpdateScript [
   &EmitterState(Life Cycle Mode=::NewEnumerator1, Loop Behavior=::NewEnumerator1, Loop Duration=3, UseLoopDelay=false)
   &SpawnBurst_Instantaneous(Spawn Count=3, Use Spawn Probability=false, UseLoopCountLimit=false)
   &SpawnRate(SpawnRate=4, Use Spawn Probability=false)
  ]
  particleSpawnScript [
   /SetVariables(Particles.VisibilityTag=&SpawnGroup()) {isSetParameters=true}
   &InitializeParticle(
    AdjustAlpha=false
    AdjustHue=false
    AdjustSaturation=false
    AdjustValue=true
    Color Mode=::NewEnumerator3
    Lifetime=&RandomRangeFloat(Minimum=2, Maximum=4)
    Mesh Scale Mode=::NewEnumerator3
    Mesh Uniform Scale=&RandomRangeFloat(Minimum=0.09, Maximum=0.19)
    UsePositionOffset=false
    Value Range:Vector2f=<1 2>
    Write Material Random=true
   )
   &InitialMeshOrientation(Orientation Axis=<0 0 1>, Orientation Vector=<0 0 1>, Rotation=<0 0.25 0>, Use Orientation Vector=true, Use Rotation Vector=true) {module=InitialMeshOrientation001}
   &ShapeLocation(Box Size=<30 0 40>, MaskSpawnGroup=false, Override Seed=false, Shape Primitive=::NewEnumerator2, Surface Only=false, UseOwnerScale=true) {module=ShapeLocation001}
  ]
  particleUpdateScript [
   &ParticleState()
   &ScaleMeshSize(Scale Factor=&VectorFromFloat(Value=&FloatFromCurve(
      FloatCurve=DataInterface() {propertyValues {Curve {keys [
          {arriveTangent=4.496, interpMode=RCIM_Cubic, leaveTangent=4.496, tangentMode=RCTM_User, time=0, value=1}
          {arriveTangent=0.625, interpMode=RCIM_Cubic, leaveTangent=0.625, time=0.3, value=1.625}
          <at 1 1.625 RCIM_Cubic>
         ]}}}
      Scale Curve=2.85
     )))
   &ScaleColor(
    Linear Color Curve=DataInterface() {propertyValues {
      AlphaCurve {keys [<at 0 0>,<at 0.1051 0.03>,<at 1 0>]}
      BlueCurve {keys [<at 0.0029 0.228>,<at 0.2356 0.2963>,<at 0.6139 0.0714>,<at 1 0.0225>]}
      GreenCurve {keys [<at 0.0029 0.7396>,<at 0.2356 0.5521>,<at 0.6139 0.2083>,<at 1 0.0417>]}
      RedCurve {keys [<at 0.0029 0.2196>,<at 0.2356 0.3657>,<at 0.6139 0.0608>,<at 1 0.008>]}
     }}
    Scale Mode=::NewEnumerator2
   )
   &DynamicMaterialParameters(
    Index 0 Param 1=&Add_Float(
     A=&FloatFromCurve(
      FloatCurve=DataInterface() {propertyValues {Curve {keys [{arriveTangent=4.8237, interpMode=RCIM_Cubic, leaveTangent=4.8237, tangentMode=RCTM_User, time=0, value=0}, <at 1 2.25 RCIM_Cubic>]}}}
      Scale Curve=0.5
     )
     B=&RandomRangeFloat(Minimum=0, Maximum=999)
    )
    Module.Index0Param0Write=true
    Module.Index0Param1Write=false
    Module.Index0Param2Write=false
    Module.Index0Param3Write=false
   )
   &Drag(Use Linear Drag=true, Use Rotational Drag=true)
   &GravityForce(Gravity=&MaskVectorBySpawnGroup(
     Masked Vector=&RandomRangeVector(Minimum=<15 15 25>, Maximum=<-15 -15 -25>)
     Passthrough Vector=<0 0 0>
    ))
   &SolveForcesAndVelocity(Clamp Velocity=false, Limit Acceleration=false)
   &CameraOffset(Camera Offset Amount=&RandomRangeFloat(Minimum=-5, Maximum=-20)) {module=CameraOffset001}
   &ScaleColor(Scale Alpha=$System.FadeOut, ScaleA=true, ScaleRGB=false) {module=ScaleColor001}
   /SetVariables(
    Particles.BackplateColor:LinearColor=&Multiply_LinearColorByFloat(
     Linear Color=&MultiplyLinearColor(
      Linear Color=<rgba 0.0579 0.0885 0.0475 0.1>
      Scale Factor=&MakeVector4_FromLinearColor(LinearColor=$System.ToD_Hue)
     )
     Float=&FloatFromCurve(
      FloatCurve=DataInterface() {propertyValues {Curve {
         defaultValue=0
         keys [
          {arriveTangent=2.3405, interpMode=RCIM_Cubic, leaveTangent=2.3405, tangentMode=RCTM_User, time=0, value=0}
          {interpMode=RCIM_Cubic, tangentMode=RCTM_User, time=0.2195, value=1}
          {tangentMode=RCTM_User, time=1, value=0}
         ]
        }}}
      Scale Curve=&Multiply_Float(A=$System.FadeOut, B=$System.ToD)
     )
    )
    Particles.Color=&MultiplyLinearColor(Linear Color=$Particles.Color, Scale Factor=&MakeVector4_FromLinearColor(LinearColor=$System.ToD_Hue))
    Particles.DMP1:Vector4f=&MakeVector4(X=0, Y=0, Z=0, W=0)
   ) {isSetParameters=true}
  ]
  rendererClasses [&NiagaraMeshRendererProperties]
  renderers [
   {rendererClass=&NiagaraMeshRendererProperties, rendererIndex=0}
   {rendererClass=&NiagaraMeshRendererProperties, rendererIndex=1}
  ]
 }}
summary {
 systemName=NS_ArcanaAgate_GhostAura_Ambient_Looping
 emitters [{emitterName=Flow001, rendererClasses [&NiagaraMeshRendererProperties]}]
 userVariables []
}
