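# Niagara system for a short green smoke poof in the ArcanaAgate emote.
# GPU emitter SmokeOut bursts sprites from a plane with a box culled from its middle, pushed by point forces and drag while they grow and fade.
# System time-of-day values scale alpha and tint both particle color and a separate BackplateColor, with two sprite renderers per emitter.
# A wider, forceless variant, SmokeOutandBack, is present but disabled.
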NiagaraSystem /BRCosmetics/Animation/Game/MainPlayer/Emotes/ArcanaAgate/FX/NS_ArcanaAgate_Poof_Cloud
dependencies {
 usedDataInterfaces [&NiagaraDataInterfaceSpriteRendererInfo, &NiagaraDataInterfaceCurve]
 usedDynamicInputs [&FNTimeOfDayScalar, &FNTimeOfDayColor]
 usedModules [
  &SystemState
  &self:SystemSpawnScript.SystemScriptSource.SystemScriptGraph.NiagaraNodeAssignment_0.SetVariables_#1
 ]
}
emitters {
 SmokeOut {
  simTarget=GPUComputeSim
  emitterUpdateScript [
   &EmitterState()
   &SpawnBurst_Instantaneous(Spawn Count=175, Use Spawn Probability=false)
  ]
  particleSpawnScript [
   &InitializeParticle(
    AdjustAlpha=true
    AdjustHue=true
    AdjustSaturation=true
    AdjustValue=false
    Alpha Scale Range:Vector2f=<0.5 1>
    Color=&UniformRangedLinearColor(Minimum=<rgba 0.3038 0.7031 0.2702 0.2>, Maximum=<rgba 0.0473 0.1042 0.0286 0.1>)
    Color Mode=::NewEnumerator3
    Hue Shift Range:Vector2f=<-0.03 0.03>
    Lifetime=&UniformRangedFloat(Minimum=0.2, Maximum=0.6)
    Sprite Rotation Mode=::NewEnumerator2
    Sprite Size Mode=::NewEnumerator1
    Sprite UV Mode=::NewEnumerator1
    Uniform Sprite Size Max=18
    Uniform Sprite Size Min=10
    UsePositionOffset=false
    Write Material Random=true
   ) {module=InitializeParticle001}
   &AddVelocityFromPoint(
    Module.Use Falloff Distance=false
    Origin Offset=<-0.5 -5 2>
    Velocity Origin:Vector3f=&SimulationPosition()
    Velocity Strength=&RandomRangeFloat(Minimum=25, Maximum=45)
   )
   &ShapeLocation(
    Box / Plane Mode:ENiagara_BoxPlaneMode=::NewEnumerator1
    MaskSpawnGroup=false
    Override Seed=false
    Plane Size:Vector2f=<30 45>
    Rotation Axis=<0 1 0>
    Rotation Mode=::NewEnumerator1
    Shape Primitive=::NewEnumerator2
    UseOwnerScale=true
   )
   &KillParticlesInVolume(Box Size:Vector3f=<15 10 30>, Kill Shape=::NewEnumerator1)
   &SubUVAnimation(
    End Frame Range Override=4
    Evaluation Type=::NewEnumerator0
    "Sprite Renderer "=DataInterface() {propertyValues {SpriteRenderer=&self:SmokeOut_0.NiagaraSpriteRendererProperties_1}}
    SubUV Animation Mode:ENiagara_SubUVLookupModeV2=::NewEnumerator2
    UseEndFrame=false
    UseStartFrame=false
   )
  ]
  particleUpdateScript [
   &ParticleState()
   &PointForce(Force Strength=300, Module.Use Falloff=false, Module.UseLinearForce=false)
   &PointForce(
    Force Strength=&RandomRangeFloat(Minimum=50, Maximum=100)
    Linear Force Vector:Vector3f=&RandomRangeVector(Minimum=<0 1 -0.7>, Maximum=<0 1 -1>)
    Module.Use Falloff=false
    Module.UseLinearForce=true
   ) {module=PointForce001}
   &Drag(
    Drag=&FloatFromCurve(
     FloatCurve=DataInterface() {propertyValues {Curve {
        defaultValue=1
        keys [
         {arriveTangent=0.064, interpMode=RCIM_Cubic, leaveTangent=0.064, tangentMode=RCTM_User, time=0, value=0}
         {arriveTangent=1.198, interpMode=RCIM_Cubic, leaveTangent=1.198, time=0.3762, value=0.1005}
         {arriveTangent=1.6386, interpMode=RCIM_Cubic, leaveTangent=1.6386, time=0.5981, value=0.7166}
         {interpMode=RCIM_Cubic, leaveTangent=-0.1237, time=0.8848, value=0.9338}
        ]
       }}}
     Scale Curve=&RandomRangeFloat(Minimum=1, Maximum=35)
    )
    Use Linear Drag=true
    Use Rotational Drag=false
   )
   &SolveForcesAndVelocity(Clamp Velocity=false, Limit Acceleration=false)
   &ScaleSpriteSize(Scale Factor=&Vector2DFromFloat(Value=&FloatFromCurve(
      FloatCurve=DataInterface() {propertyValues {Curve {
         defaultValue=0.2103
         keys [
          {arriveTangent=-0.5875, interpMode=RCIM_Cubic, leaveTangent=-0.5875, tangentMode=RCTM_User, time=0, value=0}
          {arriveTangent=3.2068, interpMode=RCIM_Cubic, leaveTangent=3.2068, tangentMode=RCTM_User, time=0.3079, value=0.3079}
          {arriveTangent=2.2818, interpMode=RCIM_Cubic, leaveTangent=2.2818, tangentMode=RCTM_User, time=0.43, value=0.8255}
          {arriveTangent=0.7293, interpMode=RCIM_Cubic, leaveTangent=0.7293, time=0.6615, value=1.1386}
          {arriveTangent=0.0224, interpMode=RCIM_Cubic, leaveTangent=0.0224, tangentMode=RCTM_User, time=1.0095, value=1.2482}
         ]
        }}}
      Scale Curve=&RandomRangeFloat(Minimum=2.6, Maximum=2.8)
     )))
   &DynamicMaterialParameters(
    Index 0 Param 1=&RandomRangeFloat(Minimum=0, Maximum=111)
    Module.Param0WriteEnabled=true
    Module.Param1WriteEnabled=false
    Module.Param2WriteEnabled=false
    Module.Param3WriteEnabled=false
   ) {module=DynamicMatParams}
   &CameraOffset(Camera Offset Amount=&RandomRangeFloat(Minimum=5, Maximum=25))
   &ScaleColor(
    Scale Alpha=&FloatFromCurve(
     FloatCurve=DataInterface() {propertyValues {Curve {
        defaultValue=0
        keys [
         {arriveTangent=1.8874, interpMode=RCIM_Cubic, leaveTangent=1.8874, tangentMode=RCTM_User, time=0, value=0}
         {arriveTangent=3.6305, interpMode=RCIM_Cubic, leaveTangent=1.8499, tangentMode=RCTM_Break, time=0.2745, value=5}
         {arriveTangent=-4.1936, interpMode=RCIM_Cubic, leaveTangent=-2.0071, tangentMode=RCTM_Break, time=0.5, value=1}
         {arriveTangent=0.4269, tangentMode=RCTM_User, time=1, value=0}
        ]
       }}}
     Scale Curve=&Multiply_Float(A=&RandomRangeFloat(Minimum=1, Maximum=1.3), B=$System.ToD)
    )
    Scale RGB=&VectorFromFloat(Value=&FloatFromCurve(FloatCurve=DataInterface() {propertyValues {Curve {keys [
          <at 0 1 RCIM_Cubic>
          {arriveTangent=-1.5024, interpMode=RCIM_Cubic, leaveTangent=-1.5024, tangentMode=RCTM_User, time=0.15, value=3}
          {arriveTangent=-5.1594, interpMode=RCIM_Cubic, leaveTangent=-5.1594, tangentMode=RCTM_User, time=0.2827, value=1.6271}
          <at 0.9999 0.5>
         ]}}}))
    ScaleA=true
    ScaleRGB=true
   )
   /SetVariables(
    Particles.BackplateColor:LinearColor=&Multiply_LinearColorByFloat(
     Linear Color=&MultiplyLinearColor(
      Linear Color=<rgba 0.0507 0.0885 0.0523 0.2>
      Scale Factor=&MakeVector4_FromLinearColor(LinearColor=$System.ToD_Hue)
     )
     Float=&FloatFromCurve(
      FloatCurve=DataInterface() {propertyValues {Curve {
         defaultValue=1
         keys [
          {arriveTangent=0.5235, interpMode=RCIM_Cubic, leaveTangent=0.5235, tangentMode=RCTM_User, time=0, value=0}
          {arriveTangent=3.2902, interpMode=RCIM_Cubic, leaveTangent=-3.4665, tangentMode=RCTM_Break, time=0.3183, value=2}
          {interpMode=RCIM_Cubic, tangentMode=RCTM_User, time=0.5, value=1}
          {arriveTangent=-2.8456, tangentMode=RCTM_User, time=1, value=0}
         ]
        }}}
      Scale Curve=&Multiply_Float(A=1, B=$System.ToD)
     )
    )
    Particles.Color=&MultiplyLinearColor(Linear Color=$Particles.Color, Scale Factor=&MakeVector4_FromLinearColor(LinearColor=$System.ToD_Hue))
   ) {isSetParameters=true, module=SetVariables_#2}
  ]
  rendererClasses [&NiagaraSpriteRendererProperties]
  renderers [
   {rendererClass=&NiagaraSpriteRendererProperties}
   {rendererClass=&NiagaraSpriteRendererProperties, rendererIndex=1}
  ]
 }
 SmokeOutandBack {
  bEnabled=false
  simTarget=GPUComputeSim
  emitterUpdateScript [
   &EmitterState()
   &SpawnBurst_Instantaneous(Spawn Count=250, Use Spawn Probability=false)
  ]
  particleSpawnScript [
   &InitializeParticle(
    AdjustAlpha=true
    AdjustHue=true
    AdjustSaturation=true
    AdjustValue=false
    Color=&UniformRangedLinearColor(Minimum=<rgba 0.3038 0.7031 0.2702 0.1>, Maximum=<rgba 0.0473 0.1042 0.0286 0.05>)
    Color Mode=::NewEnumerator3
    Hue Shift Range:Vector2f=<-0.03 0.03>
    Lifetime=&UniformRangedFloat(Minimum=0.3, Maximum=0.6)
    Sprite Rotation Mode=::NewEnumerator2
    Sprite Size Mode=::NewEnumerator1
    Sprite UV Mode=::NewEnumerator1
    Uniform Sprite Size Max=18
    Uniform Sprite Size Min=10
    UsePositionOffset=false
    Write Material Random=true
   ) {module=InitializeParticle001}
   &AddVelocityFromPoint(
    Module.Use Falloff Distance=false
    Origin Offset=&RandomRangeVector(Minimum=<-0.5 -4 2>, Maximum=<0.5 -5 2>)
    Velocity Origin:Vector3f=&SimulationPosition()
    Velocity Strength=&RandomRangeFloat(Minimum=5, Maximum=200)
   )
   &ShapeLocation(
    Box / Plane Mode:ENiagara_BoxPlaneMode=::NewEnumerator1
    MaskSpawnGroup=false
    Override Seed=false
    Plane Size:Vector2f=<55 65>
    Rotation Axis=<0 1 0>
    Rotation Mode=::NewEnumerator1
    Shape Primitive=::NewEnumerator2
    UseOwnerScale=true
   )
   &KillParticlesInVolume(Box Size:Vector3f=<40 10 50>, Kill Shape=::NewEnumerator1)
   &SubUVAnimation(
    End Frame Range Override=4
    Evaluation Type=::NewEnumerator0
    "Sprite Renderer "=DataInterface() {propertyValues {SpriteRenderer=&self:SmokeOutandBack_0.NiagaraSpriteRendererProperties_1}}
    SubUV Animation Mode:ENiagara_SubUVLookupModeV2=::NewEnumerator2
    UseEndFrame=false
    UseStartFrame=false
   )
  ]
  particleUpdateScript [
   &ParticleState()
   &Drag(
    Drag=&FloatFromCurve(
     FloatCurve=DataInterface() {propertyValues {Curve {
        defaultValue=1
        keys [{interpMode=RCIM_Cubic, tangentMode=RCTM_User, time=0, value=0}, {interpMode=RCIM_Cubic, tangentMode=RCTM_User, time=1, value=1}]
       }}}
     Scale Curve=&RandomRangeFloat(Minimum=5, Maximum=7)
    )
    Use Linear Drag=true
    Use Rotational Drag=false
   )
   &SolveForcesAndVelocity(Clamp Velocity=false, Limit Acceleration=false)
   &ScaleSpriteSize(Scale Factor=&Vector2DFromFloat(Value=&FloatFromCurve(
      FloatCurve=DataInterface() {propertyValues {Curve {
         defaultValue=0.2103
         keys [
          {arriveTangent=-0.5875, interpMode=RCIM_Cubic, leaveTangent=-0.5875, tangentMode=RCTM_User, time=0, value=0}
          {arriveTangent=2.2941, interpMode=RCIM_Cubic, leaveTangent=2.2941, tangentMode=RCTM_User, time=0.2554, value=0.3316}
          {arriveTangent=0.8977, interpMode=RCIM_Cubic, leaveTangent=0.8977, time=0.4467, value=0.7662}
          {arriveTangent=0.0224, interpMode=RCIM_Cubic, leaveTangent=0.0224, tangentMode=RCTM_User, time=1, value=1}
         ]
        }}}
      Scale Curve=2.6
     )))
   &DynamicMaterialParameters(
    Index 0 Param 1=&RandomRangeFloat(Minimum=0, Maximum=111)
    Module.Param0WriteEnabled=true
    Module.Param1WriteEnabled=false
    Module.Param2WriteEnabled=false
    Module.Param3WriteEnabled=false
   ) {module=DynamicMatParams}
   &CameraOffset(Camera Offset Amount=&RandomRangeFloat(Minimum=5, Maximum=25))
   &ScaleColor(
    Scale Alpha=&FloatFromCurve(
     FloatCurve=DataInterface() {propertyValues {Curve {
        defaultValue=0
        keys [
         {arriveTangent=1.9509, interpMode=RCIM_Cubic, leaveTangent=1.9509, tangentMode=RCTM_User, time=0, value=0}
         {interpMode=RCIM_Cubic, tangentMode=RCTM_User, time=0.5, value=1}
         {arriveTangent=-2.2601, tangentMode=RCTM_User, time=1, value=0}
        ]
       }}}
     Scale Curve=&RandomRangeFloat(Minimum=0.7, Maximum=1)
    )
    ScaleA=true
    ScaleRGB=false
   )
   /SetVariables(Particles.BackplateColor:LinearColor=&Multiply_LinearColorByFloat(
     Linear Color=<rgba 0.0498 0.0885 0.0526 0.07>
     Float=&FloatFromCurve(FloatCurve=DataInterface() {propertyValues {Curve {
         defaultValue=1
         keys [
          {arriveTangent=2.9751, interpMode=RCIM_Cubic, leaveTangent=2.9751, tangentMode=RCTM_User, time=0, value=0}
          {interpMode=RCIM_Cubic, tangentMode=RCTM_User, time=0.5, value=1}
          {arriveTangent=-2.8456, tangentMode=RCTM_User, time=1, value=0}
         ]
        }}})
    )) {isSetParameters=true, module=SetVariables_#2}
  ]
  rendererClasses [&NiagaraSpriteRendererProperties]
  renderers [
   {rendererClass=&NiagaraSpriteRendererProperties}
   {rendererClass=&NiagaraSpriteRendererProperties, rendererIndex=1}
  ]
 }
}
summary {
 systemName=NS_ArcanaAgate_Poof_Cloud
 emitters [
  {bEnabled=false, emitterName=SmokeOutandBack, simTarget=GPUComputeSim, rendererClasses [&NiagaraSpriteRendererProperties]}
  {emitterName=SmokeOut, simTarget=GPUComputeSim, rendererClasses [&NiagaraSpriteRendererProperties]}
 ]
 userVariables [
  User.bIsInWater {type=NiagaraBool}
  User.EffectScale {type=NiagaraFloat, defaultValue=1, description=""}
  User.EffectSpawnScale {type=NiagaraFloat, defaultValue=10, description=""}
  User.ForwardSpeed {type=NiagaraFloat, defaultValue=0, description=""}
  User.isBoosting {type=NiagaraBool}
  User.WaterBodyIndex {type=NiagaraFloat, defaultValue=0, description=""}
  User.WaterDepth {type=NiagaraFloat, defaultValue=0, description=""}
  User.WaterPlaneLocation {type=Vector3f, description="", defaultValue {struct=Vector3f, value=<0 0 0>}}
  User.WaterPlaneNormal {type=Vector3f, description="", defaultValue {struct=Vector3f, value=<0 0 0>}}
  User.WaveAttenuationFactor {type=NiagaraFloat, defaultValue=0, description=""}
  User.WaveTime {type=NiagaraFloat, defaultValue=0, description=""}
 ]
}
